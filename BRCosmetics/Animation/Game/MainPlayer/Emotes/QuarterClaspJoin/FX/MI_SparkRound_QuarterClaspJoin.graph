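# Translucent unlit material for round spark particles on sprites, filed under the QuarterClaspJoin emote's FX.
# Particle color RGB feeds the shared emissive adjustment function, while particle alpha multiplies a round spark texture mask.
# The mask is depth-faded on the default shading path; mobile uses the green channel directly.
# A particle-speed lerp blends the red channel with itself, so speed has no effect as wired.
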Material /BRCosmetics/Animation/Game/MainPlayer/Emotes/QuarterClaspJoin/FX/MI_SparkRound_QuarterClaspJoin
graph {
 properties {
  bEnableResponsiveAA=true
  bUsedWithNanite=true
  bUsedWithNiagaraSprites=true
  bUsedWithParticleSprites=true
  blendMode=BLEND_Translucent
  pixelDepthOffsetMode=PDOM_Legacy
  shadingModel=MSM_Unlit
  usedShadingModels=Unlit
 }
 nodes [
  n1 = ParticleColor()
  MaterialFunctionCall(Emissive=@n1.RGB) {materialFunction=&MF_FX_EmissiveAdjustments}
  n2 = TextureSample() {texture=/Game/Global/Textures/Default/Masks_and_Gradients/T_SparkRound}
  n3 = ParticleSpeed()
  n4 = ScalarParameter() {defaultValue=100, parameterName=ParticleSpeed}
  n5 = LinearInterpolate(A=@n2.R, B=@n2.R, Alpha=Clamp(None=Divide(A=@n3, B=@n4)))
  n6 = ScalarParameter() {parameterName=Fade Distance}
  Multiply(A=@n1.A, B=ShadingPathSwitch(Default=DepthFade(Opacity=@n5, FadeDistance=@n6), Mobile=@n2.G))
 ]
}
