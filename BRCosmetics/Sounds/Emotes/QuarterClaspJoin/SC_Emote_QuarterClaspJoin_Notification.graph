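# Sound cue for the QuarterClaspJoin emote's notification sound.
# A short random delay of up to a tenth of a second feeds an equally weighted random pick among five wave variations.
# It uses emote attenuation, the emote music sound class and the shared emote music concurrency rule.
# A submix send routes it to the music audio analysis submix.
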SoundCue /BRCosmetics/Sounds/Emotes/QuarterClaspJoin/SC_Emote_QuarterClaspJoin_Notification
root=Delay_20
nodes [
 {id=Delay_20, children [Random_4], properties {delayMax=0.1}}
 {id=Random_4, children [WavePlayer_20, WavePlayer_21, WavePlayer_22, WavePlayer_23, WavePlayer_24], properties {weights [1,1,1,1,1]}}
 {id=WavePlayer_20, properties {soundWaveAssetPtr=&/IP_Emote_QuarterClaspJoin_Notification_01}}
 {id=WavePlayer_21, properties {soundWaveAssetPtr=&/IP_Emote_QuarterClaspJoin_Notification_02}}
 {id=WavePlayer_22, properties {soundWaveAssetPtr=&/IP_Emote_QuarterClaspJoin_Notification_03}}
 {id=WavePlayer_23, properties {soundWaveAssetPtr=&/IP_Emote_QuarterClaspJoin_Notification_04}}
 {id=WavePlayer_24, properties {soundWaveAssetPtr=&/IP_Emote_QuarterClaspJoin_Notification_05}}
]
properties {
 attenuationSettings=&Fort_Emotes_Athena_Att
 maxDistance=1600
 soundClassObject=&Fort_Music_Emote
 concurrencyOverrides {bLimitToOwner=true, maxCount=3}
 concurrencySet [/Game/Sounds/Concurrency_Rules/Athena_EmoteMusic_Conc]
 soundSubmixSends [{minSendLevel=0, soundSubmix=&Fortnite_AudioAnalysis_Music}]
}
