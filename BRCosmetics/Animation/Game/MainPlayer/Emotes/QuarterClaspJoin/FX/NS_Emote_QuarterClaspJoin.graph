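# Niagara system for the QuarterClaspJoin emote, drawing what appears to be a livestream overlay around the character.
# Sprite emitters provide a Live badge, chat lines scrolling upward in steps, rising icons with notification audio, and a multi-renderer rolling digit counter.
# A five-group glow frame fades with camera angle.
# Front-end modifiers adjust placement and facing for menus, while system indoor colors tint several elements.
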NiagaraSystem /BRCosmetics/Animation/Game/MainPlayer/Emotes/QuarterClaspJoin/FX/NS_Emote_QuarterClaspJoin
dependencies {
 usedDataInterfaces [
  &NiagaraDataInterfaceCurve
  &NiagaraDataInterfaceCamera
  &NiagaraDataInterfaceSpriteRendererInfo
  &NiagaraDataInterfaceArrayFloat
  &NiagaraDataInterfacePlatformSet
 ]
 usedDynamicInputs [
  &Multiply_Vector4
  &MakeVector4_FromLinearColor
  &Add_Integer
  /Game/Effects/Niagara/DynamicInputs/Logic/FN_PickLinearColorDependOnPlatform
  &LerpLinearColors
 ]
 usedModules [
  &self:SystemSpawnScript.SystemScriptSource.SystemScriptGraph.NiagaraNodeAssignment_0.SetVariables_#1
  &SystemState
  &self:SetPeeloutFXArrayIndex
  &self:SystemSpawnScript.SystemScriptSource.SystemScriptGraph.NiagaraNodeAssignment_6.SetVariables_#2
  &self:SystemSpawnScript.SystemScriptSource.SystemScriptGraph.NiagaraNodeAssignment_7.SetVariables_#3
 ]
}
emitters {
 Chat {
  emitterUpdateScript [
   &EmitterState(Life Cycle Mode=::NewEnumerator1, Loop Behavior=::NewEnumerator1, Loop Duration Mode=::NewEnumerator1, UseLoopDelay=false)
   &SpawnRate(Use Spawn Probability=false)
   &SpawnBurst_Instantaneous(Use Spawn Probability=false, UseLoopCountLimit=false) {enabled=false}
   /SetVariables(Emitter.Vector:Vector3f=&MakeVector(
     X=0
     Y=0
     Z=&FloatFromCurve(
      FloatCurve=DataInterface() {propertyValues {Curve {keys [<at 0 2.7 RCIM_Constant>,<at 0.5 0 RCIM_Constant>]}}}
      CurveIndex=&ModuloFloat(A=$Emitter.Age, B=1)
     )
    )) {enabled=false, isSetParameters=true}
  ]
  particleSpawnScript [
   &PartitionParticles(Number Of Partitions=14)
   &InitializeParticle(
    Color Mode=::NewEnumerator0
    Lifetime=5
    Position Offset=<-70 0 72>
    Sprite Size=<50 25>
    Sprite Size Mode=::NewEnumerator4
    UsePositionOffset=true
    Write Material Random=true
   ) {enabled=false}
   &InitializeParticle(
    Lifetime=7
    Position Offset=<-0.82 0 8>
    Sprite Size=<58 10>
    Sprite Size Mode=::NewEnumerator4
    UsePositionOffset=true
    Write Material Random=true
   ) {module=InitializeParticle001}
   &SpriteFacingAndAlignment(
    Alignment Coordinate Space=::Local
    Sprite Alignment=<0 0 1>
    Sprite Facing=&FN_FrontEndVector3Modifier(Gameplay=<0 -1 0>, Front End=<0 -1 0.04>)
    WriteAlignment=true
    WriteFacing=true
   ) {module=SpriteFacingAndAlignment001}
   &AddVelocity(Velocity=<0 0 0.5>)
   &CameraOffset(Camera Offset Amount=&FN_FrontEndFloatModifier(Gameplay=0, Front End=20)) {enabled=false}
  ]
  particleUpdateScript [
   &ParticleState()
   &/InitScreenSpaceSprite()
   &SubUVAnimation(
    "Sprite Renderer "=DataInterface() {propertyValues {SpriteRenderer=&self:Chat_0.NiagaraSpriteRendererProperties_0}}
    SubUV Animation Mode:ENiagara_SubUVLookupModeV2=::NewEnumerator5
   ) {enabled=false, module=SubUVAnimation002}
   &SubUVAnimation(
    Direct Frame Index:NiagaraFloat=&SelectFloatFromArray(
     Direct Array Index=$Particles.PartitionParticles.Partition
     Float Selection Array=DataInterface() {propertyValues {FloatData [0,1,2,3,4,5,6,7,8,9,10,11,12,13]}}
    )
    "Sprite Renderer "=DataInterface() {propertyValues {SpriteRenderer=&self:Chat_0.NiagaraSpriteRendererProperties_0}}
    SubUV Animation Mode:ENiagara_SubUVLookupModeV2=::NewEnumerator5
   )
   &ScaleVelocity(Velocity Scale=&MakeVector(
     X=0
     Y=0
     Z=&FloatFromCurve(
      FloatCurve=DataInterface() {propertyValues {Curve {keys [<at 0 1 RCIM_Constant>,<at 0.5 0 RCIM_Constant>,<at 0.9 1 RCIM_Constant>,<at 1.5 0 RCIM_Constant>,<at 1.9 1 RCIM_Constant>,<at 2.5 0 RCIM_Constant>,<at 2.87 1 RCIM_Constant>,<at 3.5 0 RCIM_Constant>,<at 3.89 1 RCIM_Constant>,<at 4.5 0 RCIM_Constant>,<at 4.87 1 RCIM_Constant>,<at 5.5 0 RCIM_Constant>,<at 5.85 1 RCIM_Constant>,<at 6.5 0 RCIM_Constant>,<at 7 1 RCIM_Constant>]}}}
      CurveIndex=$Particles.Age
      Scale Curve=3.3
     )
    ))
   &SolveForcesAndVelocity(Clamp Velocity=false, Limit Acceleration=false)
   &ScaleColor(Scale RGB=&MakeVectorFromLinearColorRGB(Color=$System.IndoorColor), ScaleA=true, ScaleRGB=true) {module=ScaleColor004}
   &/ScreenSpaceSprite()
   &ScaleColor(
    Scale Alpha=&FloatFromCurve(FloatCurve=DataInterface() {propertyValues {Curve {
        defaultValue=1
        keys [
         <at 0 0 RCIM_Cubic>
         {arriveTangentWeight=0.0612, interpMode=RCIM_Cubic, leaveTangentWeight=0.3364, tangentMode=RCTM_User, time=0.1, value=1}
         {interpMode=RCIM_Cubic, tangentMode=RCTM_User, time=0.75, value=1}
         {interpMode=RCIM_Cubic, tangentMode=RCTM_User, time=1, value=0}
        ]
       }}})
    ScaleA=true
    ScaleRGB=true
   )
   /SetVariables(Particles.Position=&AddVectorToPosition(
     Position=$Particles.Position
     Vector=&FN_FrontEndVector3Modifier(Gameplay=<0 0 0>, Front End=<0 -5 0>)
    )) {isSetParameters=true}
  ]
  rendererClasses [&NiagaraSpriteRendererProperties]
  renderers [{rendererClass=&NiagaraSpriteRendererProperties}]
 }
 Eye_Glow_Frame001 {
  emitterUpdateScript [
   &EmitterState(Life Cycle Mode=::NewEnumerator1, Loop Behavior=::NewEnumerator1, Loop Duration Mode=::NewEnumerator1, UseLoopDelay=false)
   &SpawnBurst_Instantaneous(Use Spawn Probability=false, UseLoopCountLimit=false)
   &SpawnBurst_Instantaneous(Spawn Group=1, Use Spawn Probability=false, UseLoopCountLimit=false) {module=SpawnBurst_Instantaneous001}
   &SpawnBurst_Instantaneous(Spawn Group=2, Use Spawn Probability=false, UseLoopCountLimit=false) {module=SpawnBurst_Instantaneous002}
   &SpawnBurst_Instantaneous(Spawn Group=3, Use Spawn Probability=false, UseLoopCountLimit=false) {module=SpawnBurst_Instantaneous003}
   &SpawnBurst_Instantaneous(Spawn Group=4, Use Spawn Probability=false, UseLoopCountLimit=false) {module=SpawnBurst_Instantaneous004}
  ]
  particleSpawnScript [
   /SetVariables(Particles.VisibilityTag=&SpawnGroup()) {isSetParameters=true}
   &InitializeParticle(
    Color=&MergeHelper_ColorSwitch(
     0=<rgba 1 1 1 1>
     1=<rgba 1 0 0 1>
     2=<rgba 1 0 0 0.5>
     3=<rgba 1 0 0 0.5>
     4=<rgba 1 1 1 1>
     Index=&SpawnGroup()
     Num=::NewEnumerator3
    )
    Lifetime=0.7
    Position Offset=&MergeHelper_VectorSwitch(
     0=&FN_FrontEndVector3Modifier(Gameplay=<-60 0 180>, Front End=<-60 2 180>)
     1=<0 0 190>
     2=&FN_FrontEndVector3Modifier(Gameplay=<102 0 135>, Front End=<102 0 125>)
     3=&FN_FrontEndVector3Modifier(Gameplay=<-102 0 135>, Front End=<-102 0 125>)
     4=<0 0 132>
     Index=&SpawnGroup()
     Num=::NewEnumerator3
    )
    Position Offset Coordinate Space=::Local
    Sprite Size=&MergeHelper_Vector2Switch(0=<8.5 8.5>, 1=<300 12>, 2=<15 200>, 3=<15 200>, 4=<210 210>, Index=&SpawnGroup(), Num=::NewEnumerator3)
    Sprite Size Mode=::NewEnumerator4
    UsePositionOffset=true
    Write Material Random=true
   )
   &SpriteFacingAndAlignment(
    Alignment Coordinate Space=::Local
    Facing Coordinate Space=::Local
    Sprite Alignment=&MergeHelper_VectorSwitch(
     0=<0 0 1>
     1=<0 1 0>
     2=&FN_FrontEndVector3Modifier(Gameplay=<0 0 1>, Front End=<0 0.04 1>)
     3=&FN_FrontEndVector3Modifier(Gameplay=<0 0 1>, Front End=<0 0.04 1>)
     4=<0 0 1>
     Index=&SpawnGroup()
     Num=::NewEnumerator3
    )
    Sprite Facing=&MergeHelper_VectorSwitch(
     0=&FN_FrontEndVector3Modifier(Gameplay=<0 -1 0>, Front End=<0 -1 0.04>)
     1=<0 0 1>
     2=<1 0 0>
     3=<1 0 0>
     4=&FN_FrontEndVector3Modifier(Gameplay=<0 -1 0>, Front End=<0 -1 0.04>)
     Index=&SpawnGroup()
     Num=::NewEnumerator3
    )
    WriteAlignment=true
    WriteFacing=true
   ) {module=SpriteFacingAndAlignment001}
   /SetVariables(Particles.PivotOffset:Vector2f=<-6.5 21.7>) {enabled=false, isSetParameters=true}
  ]
  particleUpdateScript [
   &ParticleState(Kill Particles When Lifetime Has Elapsed=false)
   &ScaleColor(
    Scale Alpha=&MergeHelper_FloatSwitch(
     0=1
     1=&ClampFloat(Float=&RemapRange(
       Input Value To Remap=&DotProduct(Vector A=&ReturnCameraVector(Camera Vectors=::NewEnumerator1, Coordinate Space=::World), Vector B=<0 0 -1>)
       Input Min=0.2
       Input Max=1
       Output Min=0
       Output Max=0.5
      ))
     2=&ClampFloat(Float=&RemapRange(
       Input Value To Remap=&Abs_Float(Float:NiagaraFloat=&DotProduct(
         Vector A=&ReturnCameraVector(Camera Vectors=::NewEnumerator1, Coordinate Space=::Simulation)
         Vector B=<0 0 -1>
        ))
       Input Min=0.8
       Input Max=0
       Output Min=0
       Output Max=1
      ))
     3=&ClampFloat(Float=&RemapRange(
       Input Value To Remap=&Abs_Float(Float:NiagaraFloat=&DotProduct(
         Vector A=&ReturnCameraVector(Camera Vectors=::NewEnumerator1, Coordinate Space=::Simulation)
         Vector B=<0 0 -1>
        ))
       Input Min=0.8
       Input Max=0
       Output Min=0
       Output Max=1
      ))
     4=&ClampFloat(Float=&RemapRange(
       Input Value To Remap=&DotProduct(Vector A=&ReturnCameraVector(Camera Vectors=::NewEnumerator1, Coordinate Space=::World), Vector B=<0 0 -1>)
       Input Min=1
       Input Max=0.4
       Output Min=0.4
       Output Max=1
      ))
     Index=&SpawnGroup()
     Num=::NewEnumerator3
    )
    ScaleA=true
    ScaleRGB=true
   )
   &ScaleColor(
    Scale Alpha=&MaskFloat(
     Mask Source=0
     Mask Parameter=?
     Mask=2
     Masked Float=&ClampFloat(
      Float=&RemapRange(
       Input Value To Remap=&Abs_Float(Float:NiagaraFloat=&DotProduct(Vector A=&ReturnCameraVector(Camera Vectors=::NewEnumerator1, Coordinate Space=::Simulation), Vector B=<0 1 0>))
       Input Min=0.3
       Input Max=0
       Output Min=0
       Output Max=1
      )
      Min=-1
     )
     Passthrough Float=1
    )
    ScaleA=true
    ScaleRGB=true
   ) {module=ScaleColor001}
   &ScaleColor(
    Scale Alpha=&MaskFloat(
     Mask Source=0
     Mask Parameter=?
     Mask=3
     Masked Float=&ClampFloat(
      Float=&RemapRange(
       Input Value To Remap=&Abs_Float(Float:NiagaraFloat=&DotProduct(Vector A=&ReturnCameraVector(Camera Vectors=::NewEnumerator1, Coordinate Space=::Simulation), Vector B=<0 1 0>))
       Input Min=0.5
       Input Max=0
       Output Min=0
       Output Max=1
      )
      Min=-1
     )
     Passthrough Float=1
    )
    ScaleA=true
    ScaleRGB=true
   ) {module=ScaleColor002}
   &ScaleColor(
    Scale Alpha=&FloatFromCurve(FloatCurve=DataInterface() {propertyValues {Curve {
        defaultValue=0
        keys [{interpMode=RCIM_Cubic, tangentMode=RCTM_User, time=0, value=0}, {interpMode=RCIM_Cubic, tangentMode=RCTM_User, time=0.7, value=1}]
       }}})
    Scale RGB=&Multiply_Vector(A=<1 1 1>, B=&MakeVectorFromLinearColorRGB(Color=$System.SildingColor))
    ScaleA=true
    ScaleRGB=true
   ) {enabled=false, module=ScaleColor003}
   &ScaleColor(
    Scale Alpha=&FloatFromCurve(FloatCurve=DataInterface() {propertyValues {Curve {
        defaultValue=0
        keys [{interpMode=RCIM_Cubic, tangentMode=RCTM_User, time=0, value=0}, {interpMode=RCIM_Cubic, tangentMode=RCTM_User, time=0.7, value=1}]
       }}})
    Scale RGB=&MergeHelper_VectorSwitch(
     0=&MakeVectorFromLinearColorRGB(Color=$System.IndoorColor)
     1=&MakeVectorFromLinearColorRGB(Color=$System.IndoorColorSide)
     2=&MakeVectorFromLinearColorRGB(Color=$System.IndoorColorSide)
     3=&MakeVectorFromLinearColorRGB(Color=$System.IndoorColorSide)
     4=&MakeVectorFromLinearColorRGB(Color=$System.IndoorColor)
     Index=&SpawnGroup()
     Num=::NewEnumerator3
    )
    ScaleA=true
    ScaleRGB=true
   ) {enabled=false, module=ScaleColor004}
   &ScaleColor(
    Scale Alpha=&FloatFromCurve(FloatCurve=DataInterface() {propertyValues {Curve {
        defaultValue=0
        keys [{interpMode=RCIM_Cubic, tangentMode=RCTM_User, time=0, value=0}, {interpMode=RCIM_Cubic, tangentMode=RCTM_User, time=0.7, value=1}]
       }}})
    Scale RGB=&MergeHelper_VectorSwitch(
     0=<1 1 1>
     1=&MakeVectorFromLinearColorRGB(Color=$System.IndoorColorSide)
     2=&MakeVectorFromLinearColorRGB(Color=$System.IndoorColorSide)
     3=&MakeVectorFromLinearColorRGB(Color=$System.IndoorColorSide)
     4=<1 1 1>
     Index=&SpawnGroup()
     Num=::NewEnumerator3
    )
    ScaleA=true
    ScaleRGB=true
   ) {module=ScaleColor005}
  ]
  rendererClasses [&NiagaraSpriteRendererProperties]
  renderers [
   {rendererClass=&NiagaraSpriteRendererProperties}
   {rendererClass=&NiagaraSpriteRendererProperties, rendererIndex=1}
   {rendererClass=&NiagaraSpriteRendererProperties, rendererIndex=2}
   {rendererClass=&NiagaraSpriteRendererProperties, rendererIndex=3}
   {rendererClass=&NiagaraSpriteRendererProperties, rendererIndex=4}
  ]
 }
 Icon_and_Dot002 {
  emitterSpawnScript [&InitializeAudioPlayer(Audio Player=DataInterface() {propertyValues {
      Attenuation=null
      Concurrency=null
      MaxPlaysPerTick=10
      SoundToPlay=/BRCosmetics/Sounds/Emotes/QuarterClaspJoin/SC_Emote_QuarterClaspJoin_Notification
      bAllowLoopingOneShotSounds=false
      bLimitPlaysPerTick=true
      bOnlyActiveDuringGameplay=false
      bStopWhenComponentIsDestroyed=true
      ConfigurationUserParameter {parameter {name=None, typeDefHandle {registeredTypeIndex=1}, varData []}}
      ParameterNames []
     }})]
  emitterUpdateScript [
   &EmitterState()
   &SpawnRate(SpawnRate=5, Use Spawn Probability=false) {enabled=false}
   &SpawnRate(
    Spawn Probability=0.65
    SpawnRate=&Sine(Normalized Angle=$Emitter.Age, Period=3, Scale=1, Bias=&RandomRangeFloat(Minimum=4, Maximum=7.5))
    Use Spawn Probability=true
   ) {module=SpawnRate001}
  ]
  particleSpawnScript [
   &InitializeParticle(
    Color Mode=::NewEnumerator0
    Lifetime=&RandomRangeFloat(Minimum=1.3, Maximum=1.55)
    Position Offset=<0 100 0>
    Sprite Size=<9 6.5>
    Sprite Size Mode=::NewEnumerator4
    UsePositionOffset=false
    Write Material Random=true
   )
   &SpriteFacingAndAlignment(
    Alignment Coordinate Space=::Local
    Sprite Alignment=<0 0 1>
    Sprite Facing=&FN_FrontEndVector3Modifier(Gameplay=<0 -1 0>, Front End=<0 -1 0.04>)
    WriteAlignment=true
    WriteFacing=true
   ) {module=SpriteFacingAndAlignment001}
   &AddVelocity(
    Cone Angle=15
    Cone Axis=<0 0 1>
    Inner Cone Angle=1
    Module.UseConeFalloff=false
    Module.UseVelDistribution=false
    Override Seed=false
    Velocity Mode=::NewEnumerator2
    Velocity Speed=10
   )
   &ShapeLocation(MaskSpawnGroup=false, Offset=<7.7 0 13>, Offset Mode=::NewEnumerator0, Override Seed=false, Sphere Radius=1, UseOwnerScale=true)
   /Niagara/Modules/Audio/PlayPersistentAudio()
   &SystemLocation(MaskSpawnGroup=false, Offset=<8 50 12>) {enabled=false}
  ]
  particleUpdateScript [
   &ParticleState()
   &InitScreenSpaceSprite()
   &Drag(Use Linear Drag=true, Use Rotational Drag=true) {enabled=false}
   &ScaleColor(
    Scale Alpha=&FloatFromCurve(FloatCurve=DataInterface() {propertyValues {Curve {
        defaultValue=1
        keys [
         {interpMode=RCIM_Cubic, tangentMode=RCTM_User, time=0, value=0}
         {interpMode=RCIM_Cubic, tangentMode=RCTM_User, time=0.5, value=1}
         {tangentMode=RCTM_User, time=1, value=0}
        ]
       }}})
    ScaleA=true
    ScaleRGB=false
   ) {enabled=false}
   &SubUVAnimation(
    Evaluation Type=::NewEnumerator0
    "Sprite Renderer "=DataInterface() {propertyValues {SpriteRenderer=&self:Icon_and_Dot002_1.NiagaraSpriteRendererProperties_0}}
    SubUV Animation Mode:ENiagara_SubUVLookupModeV2=::NewEnumerator2
    UseEndFrame=false
    UseStartFrame=false
   )
   &ScaleColor(Scale RGB=&MakeVectorFromLinearColorRGB(Color=$System.IndoorColor), ScaleA=true, ScaleRGB=true) {enabled=false, module=ScaleColor004}
   &ScaleSpriteSize(Uniform Curve Sprite Scale=DataInterface() {propertyValues {Curve {
       defaultValue=0
       keys [
        {arriveTangent=2.3881, interpMode=RCIM_Cubic, leaveTangent=2.3881, tangentMode=RCTM_User, time=0, value=0}
        {arriveTangent=7.2079, interpMode=RCIM_Cubic, leaveTangent=7.2079, time=0.2038, value=0.6822}
        {arriveTangent=-0.0618, interpMode=RCIM_Cubic, leaveTangent=-0.0618, tangentMode=RCTM_User, time=0.2915, value=2.101}
        <at 0.3276 1.7095 RCIM_Cubic>
       ]
      }}})
   &SolveForcesAndVelocity(Clamp Velocity=false, Limit Acceleration=false)
   &ScreenSpaceSprite()
   /Niagara/Modules/Audio/UpdatePersistentAudio()
   /SetVariables(Particles.Position=&AddVectorToPosition(
     Position=$Particles.Position
     Vector=&FN_FrontEndVector3Modifier(Gameplay=<0 0 0>, Front End=<0 -5 0>)
    )) {isSetParameters=true}
  ]
  rendererClasses [&NiagaraSpriteRendererProperties]
  renderers [{rendererClass=&NiagaraSpriteRendererProperties}]
 }
 Live {
  emitterSpawnScript [&InitializeAudioPlayer(Audio Player=DataInterface() {propertyValues {
      Attenuation=null
      Concurrency=null
      MaxPlaysPerTick=10
      SoundToPlay=&SC_Emote_QuarterClaspJoin_Notification
      bAllowLoopingOneShotSounds=false
      bLimitPlaysPerTick=true
      bOnlyActiveDuringGameplay=false
      bStopWhenComponentIsDestroyed=true
      ConfigurationUserParameter {parameter {name=None, typeDefHandle {registeredTypeIndex=1}, varData []}}
      ParameterNames []
     }})]
  emitterUpdateScript [
   &EmitterState(Life Cycle Mode=::NewEnumerator1, Loop Behavior=::NewEnumerator1, Loop Duration=0, UseLoopDelay=false)
   &SpawnBurst_Instantaneous(Use Spawn Probability=false, UseLoopCountLimit=false)
  ]
  particleSpawnScript [
   /SetVariables(Particles.VisibilityTag=&SpawnGroup()) {isSetParameters=true}
   &InitializeParticle(
    Color Mode=::NewEnumerator0
    Lifetime=1.7
    Position Offset=&FN_FrontEndVector3Modifier(Gameplay=<0 0 0>, Front End=<0 -5.5 0>)
    Position Offset Coordinate Space=::Local
    Sprite Size=<16.5 9>
    Sprite Size Mode=::NewEnumerator4
    UsePositionOffset=true
    Write Material Random=true
   )
   &SpriteFacingAndAlignment(
    Alignment Coordinate Space=::Local
    Sprite Alignment=<0 0 1>
    Sprite Facing=&FN_FrontEndVector3Modifier(Gameplay=<0 -1 0>, Front End=<0 -1 0.04>)
    WriteAlignment=true
    WriteFacing=true
   ) {module=SpriteFacingAndAlignment001}
   /SetVariables(Particles.PivotOffset:Vector2f=<6.05 20.8>) {isSetParameters=true}
   &PlayPersistentAudio() {enabled=false}
  ]
  particleUpdateScript [
   &ParticleState(Kill Particles When Lifetime Has Elapsed=false)
   &SubUVAnimation(
    Evaluation Type=::NewEnumerator0
    "Sprite Renderer "=DataInterface() {propertyValues {SpriteRenderer=&self:Live_0.NiagaraSpriteRendererProperties_0}}
    SubUV Animation Mode:ENiagara_SubUVLookupModeV2=::NewEnumerator2
    UseEndFrame=false
    UseStartFrame=false
   )
   &SolveForcesAndVelocity(Clamp Velocity=false, Limit Acceleration=false)
   &UpdatePersistentAudio() {enabled=false}
   &ScaleColor(Scale RGB=&MakeVectorFromLinearColorRGB(Color=$System.IndoorColor), ScaleA=true, ScaleRGB=true) {module=ScaleColor004}
  ]
  rendererClasses [&NiagaraSpriteRendererProperties]
  renderers [{rendererClass=&NiagaraSpriteRendererProperties}]
 }
 Numbers {
  emitterSpawnScript [
   /SetVariables(
    Emitter.Overall Offset=DataInterface() {propertyValues {Curve {keys [
        {arriveTangent=0.3227, time=1, value=0.5}
        <at 2 4>
        <at 3 3.8>
        <at 4 3.5>
        <at 5 3>
        <at 6 2.35>
        <at 7 1.85>
        <at 8 1.25>
        <at 9 0.75>
        <at 10 0>
       ]}}}
    Emitter.Space Curve=DataInterface() {propertyValues {Curve {keys [{arriveTangent=0.3227, time=1, value=0}, <at 2 1>, <at 3 1>, <at 4 1>, <at 5 1>, <at 6 1>, <at 7 1>, <at 8 1>, <at 9 1>, <at 10 1>]}}}
   ) {isSetParameters=true}
   &InitializeAudioPlayer()
  ]
  emitterUpdateScript [
   &EmitterState()
   &SpawnRate(
    Spawn Probability=&MakeCustomFloatFromBool(Bool=&SetBoolByFloatComparison(A=$Emitter.LoopedAge, B=899), True Float=0, False Float=1)
    SpawnRate=&Waveform("Waveform [1]"=::NewEnumerator0, "[1] Frequency"=1, "[1] Amplitude Min/Max"=<2 2>)
    Use Spawn Probability=true
   )
  ]
  particleSpawnScript [
   &InitializeParticle(
    Lifetime=0.5
    Mesh Scale Mode=::NewEnumerator3
    Position Offset=&FN_FrontEndVector3Modifier(Gameplay=<5 0 117>, Front End=<5 4.5 117>)
    Sprite Size Mode=::NewEnumerator3
    Uniform Sprite Size=6
    UsePositionOffset=true
    Write Material Random=true
   ) {module=InitializeParticle001}
   /SetVariables(Particles.RollingDigits:NiagaraInt32=&MaskIntBySpawnGroup(
     Masked Int=$System.MaxNumRange
     Passthrough Int=&MakeIntFromFloat(Float=&FloatFromCurve(
       FloatCurve=DataInterface() {propertyValues {Curve {
          defaultValue=0.12
          keys [
           {arriveTangent=0.2895, interpMode=RCIM_Cubic, leaveTangent=0.2895, tangentMode=RCTM_User, time=0, value=0.0408}
           {arriveTangent=0.352, tangentMode=RCTM_User, time=5, value=1}
          ]
         }}}
       CurveIndex=$Emitter.NormalizedLoopAge
       Scale Curve=&MakeFloatFromInt(Int=$System.MaxNumRange)
      ))
     Spawn Group Mask=1
    )) {isSetParameters=true}
   /SetVariables(Particles.OffsetCorner:Vector2f=<-14 10.5>) {isSetParameters=true}
   &SpriteFacingAndAlignment(
    Alignment Coordinate Space=::Local
    Sprite Alignment=<0 0 1>
    Sprite Facing=&FN_FrontEndVector3Modifier(Gameplay=<0 -1 0>, Front End=<0 -1 -0.04>)
    WriteAlignment=true
    WriteFacing=true
   ) {module=SpriteFacingAndAlignment001}
   &PlayPersistentAudio()
  ]
  particleUpdateScript [
   &ParticleState()
   &self:NewCalcDigits_04(MaximumNumber:NiagaraInt32=$Particles.RollingDigits) {module=NewCalcDigits_04}
   /SetVariables(
    Particles.PivotOffset 1:Vector2f=/Niagara/DynamicInputs/Add/Add_Vector2(
     A:Vector2f=$Particles.OffsetCorner
     B:Vector2f=&MakeVector2D(
      X=&Add_Float(
       A=&FloatFromCurve(FloatCurve=$Emitter.Overall Offset, CurveIndex=$Particles.Testdigits)
       B=&Lerp_Float(A=0, B=-4.5, Alpha=&FloatFromCurve(FloatCurve=$Emitter.Space Curve, CurveIndex=$Particles.Testdigits))
      )
      Y=0.5
     )
    )
    Particles.PivotOffset 10:Vector2f=&Add_Vector2(
     A:Vector2f=$Particles.OffsetCorner
     B:Vector2f=&MakeVector2D(
      X=&Add_Float(
       A=&FloatFromCurve(FloatCurve=$Emitter.Overall Offset, CurveIndex=$Particles.Testdigits)
       B=&Lerp_Float(A=0, B=4.5, Alpha=&FloatFromCurve(FloatCurve=$Emitter.Space Curve, CurveIndex=$Particles.Testdigits))
      )
      Y=0.5
     )
    )
    Particles.PivotOffset 2:Vector2f=&Add_Vector2(
     A:Vector2f=$Particles.OffsetCorner
     B:Vector2f=&MakeVector2D(
      X=&Add_Float(
       A=&FloatFromCurve(FloatCurve=$Emitter.Overall Offset, CurveIndex=$Particles.Testdigits)
       B=&Lerp_Float(A=0, B=-3.5, Alpha=&FloatFromCurve(FloatCurve=$Emitter.Space Curve, CurveIndex=$Particles.Testdigits))
      )
      Y=0.5
     )
    )
    Particles.PivotOffset 3:Vector2f=&Add_Vector2(
     A:Vector2f=$Particles.OffsetCorner
     B:Vector2f=&MakeVector2D(
      X=&Add_Float(
       A=&FloatFromCurve(FloatCurve=$Emitter.Overall Offset, CurveIndex=$Particles.Testdigits)
       B=&Lerp_Float(A=0, B=-2.5, Alpha=&FloatFromCurve(FloatCurve=$Emitter.Space Curve, CurveIndex=$Particles.Testdigits))
      )
      Y=0.5
     )
    )
    Particles.PivotOffset 4:Vector2f=&Add_Vector2(
     A:Vector2f=$Particles.OffsetCorner
     B:Vector2f=&MakeVector2D(
      X=&Add_Float(
       A=&FloatFromCurve(FloatCurve=$Emitter.Overall Offset, CurveIndex=$Particles.Testdigits)
       B=&Lerp_Float(A=0, B=-1.5, Alpha=&FloatFromCurve(FloatCurve=$Emitter.Space Curve, CurveIndex=$Particles.Testdigits))
      )
      Y=0.5
     )
    )
    Particles.PivotOffset 5:Vector2f=&Add_Vector2(
     A:Vector2f=$Particles.OffsetCorner
     B:Vector2f=&MakeVector2D(
      X=&Add_Float(
       A=&FloatFromCurve(FloatCurve=$Emitter.Overall Offset, CurveIndex=$Particles.Testdigits)
       B=&Lerp_Float(A=0, B=-0.5, Alpha=&FloatFromCurve(FloatCurve=$Emitter.Space Curve, CurveIndex=$Particles.Testdigits))
      )
      Y=0.5
     )
    )
    Particles.PivotOffset 6:Vector2f=&Add_Vector2(
     A:Vector2f=$Particles.OffsetCorner
     B:Vector2f=&MakeVector2D(
      X=&Add_Float(
       A=&FloatFromCurve(FloatCurve=$Emitter.Overall Offset, CurveIndex=$Particles.Testdigits)
       B=&Lerp_Float(A=0, B=0.5, Alpha=&FloatFromCurve(FloatCurve=$Emitter.Space Curve, CurveIndex=$Particles.Testdigits))
      )
      Y=0.5
     )
    )
    Particles.PivotOffset 7:Vector2f=&Add_Vector2(
     A:Vector2f=$Particles.OffsetCorner
     B:Vector2f=&MakeVector2D(
      X=&Add_Float(
       A=&FloatFromCurve(FloatCurve=$Emitter.Overall Offset, CurveIndex=$Particles.Testdigits)
       B=&Lerp_Float(A=0, B=1.5, Alpha=&FloatFromCurve(FloatCurve=$Emitter.Space Curve, CurveIndex=$Particles.Testdigits))
      )
      Y=0.5
     )
    )
    Particles.PivotOffset 8:Vector2f=&Add_Vector2(
     A:Vector2f=$Particles.OffsetCorner
     B:Vector2f=&MakeVector2D(
      X=&Add_Float(
       A=&FloatFromCurve(FloatCurve=$Emitter.Overall Offset, CurveIndex=$Particles.Testdigits)
       B=&Lerp_Float(A=0, B=2.5, Alpha=&FloatFromCurve(FloatCurve=$Emitter.Space Curve, CurveIndex=$Particles.Testdigits))
      )
      Y=0.5
     )
    )
    Particles.PivotOffset 9:Vector2f=&Add_Vector2(
     A:Vector2f=$Particles.OffsetCorner
     B:Vector2f=&MakeVector2D(
      X=&Add_Float(
       A=&FloatFromCurve(FloatCurve=$Emitter.Overall Offset, CurveIndex=$Particles.Testdigits)
       B=&Lerp_Float(A=0, B=3.5, Alpha=&FloatFromCurve(FloatCurve=$Emitter.Space Curve, CurveIndex=$Particles.Testdigits))
      )
      Y=0.5
     )
    )
   ) {isSetParameters=true}
   &UpdatePersistentAudio() {module=UpdatePersistentAudio001}
  ]
  rendererClasses [&NiagaraSpriteRendererProperties]
  renderers [
   {rendererClass=&NiagaraSpriteRendererProperties}
   {rendererClass=&NiagaraSpriteRendererProperties, rendererIndex=1}
   {rendererClass=&NiagaraSpriteRendererProperties, rendererIndex=2}
   {rendererClass=&NiagaraSpriteRendererProperties, rendererIndex=3}
   {rendererClass=&NiagaraSpriteRendererProperties, rendererIndex=4}
   {rendererClass=&NiagaraSpriteRendererProperties, rendererIndex=5}
   {rendererClass=&NiagaraSpriteRendererProperties, rendererIndex=6}
   {rendererClass=&NiagaraSpriteRendererProperties, rendererIndex=7}
   {rendererClass=&NiagaraSpriteRendererProperties, rendererIndex=8}
   {rendererClass=&NiagaraSpriteRendererProperties, rendererIndex=9}
  ]
 }
}
summary {
 systemName=NS_Emote_QuarterClaspJoin
 emitters [
  {emitterName=Numbers, rendererClasses [&NiagaraSpriteRendererProperties]}
  {emitterName=Icon_and_Dot002, rendererClasses [&NiagaraSpriteRendererProperties]}
  {emitterName=Live, rendererClasses [&NiagaraSpriteRendererProperties]}
  {emitterName=Chat, rendererClasses [&NiagaraSpriteRendererProperties]}
  {emitterName=Eye_Glow_Frame001, rendererClasses [&NiagaraSpriteRendererProperties]}
 ]
 userVariables [
  User.AudioPlayer {
   type=NiagaraDataInterfaceAudioPlayer
   description=""
   defaultValue {
    struct=NiagaraExt_VariableValue_DataInterface
    value {
     dataInterface=&self:NiagaraDataInterfaceAudioPlayer_0
     dataInterfaceClass=&NiagaraDataInterfaceAudioPlayer
    }
   }
  }
  User.bIsFrontEnd {type=NiagaraBool}
  User.bIsFrontEndPreview {type=NiagaraBool}
  User.CameraOffset {type=NiagaraFloat, defaultValue=0, description=""}
  User.IsSpawnDust {type=NiagaraBool, defaultValue=true}
  User.MaximumNumber {type=NiagaraInt32, defaultValue=99999}
  User.SlidingColorArray {
   type=NiagaraDataInterfaceArrayColor
   description=""
   defaultValue {
    struct=NiagaraExt_VariableValue_DataInterface
    value {
     dataInterface=&self:NiagaraDataInterfaceArrayColor_0
     dataInterfaceClass=&NiagaraDataInterfaceArrayColor
    }
   }
  }
  User.SurfaceTypeIndex {type=NiagaraInt32}
 ]
}
