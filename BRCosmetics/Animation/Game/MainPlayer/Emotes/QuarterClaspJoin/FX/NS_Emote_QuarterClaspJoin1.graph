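# Niagara system for the QuarterClaspJoin emote that shows a rolling number counter as camera-oriented sprites.
# One emitter spawns short-lived particles whose number follows a loop-age curve up to System.MaxNumRange, shifting from pale cyan to gold.
# A DoOnce near loop end spawns a longer-lived gold particle showing the maximum.
# NewCalcDigits processes the number for eleven sprite renderers; motion, scaling, pivot layout, heart and audio modules are disabled.
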NiagaraSystem /BRCosmetics/Animation/Game/MainPlayer/Emotes/QuarterClaspJoin/FX/NS_Emote_QuarterClaspJoin1
dependencies {
 usedDataInterfaces [&NiagaraDataInterfaceCurve, &NiagaraDataInterfaceCamera]
 usedDynamicInputs [&MakeCustomIntFromBool, &Add_Integer]
 usedModules [
  &self:SystemSpawnScript.SystemScriptSource.SystemScriptGraph.NiagaraNodeAssignment_0.SetVariables_#1
  &SystemState
 ]
}
emitters {Counter_nosim_infDigits001 {
  emitterSpawnScript [/SetVariables(
    Emitter.Overall Offset=DataInterface() {propertyValues {Curve {keys [
        {arriveTangent=0.3227, time=1, value=0.5}
        <at 2 4>
        <at 3 3.8>
        <at 4 3.5>
        <at 5 3>
        <at 6 2.35>
        <at 7 1.85>
        <at 8 1.25>
        <at 9 0.75>
        <at 10 0>
       ]}}}
    Emitter.Space Curve=DataInterface() {propertyValues {Curve {keys [{arriveTangent=0.3227, time=1, value=0}, <at 2 1>, <at 3 1>, <at 4 1>, <at 5 1>, <at 6 1>, <at 7 1>, <at 8 1>, <at 9 1>, <at 10 1>]}}}
   ) {isSetParameters=true}]
  emitterUpdateScript [
   &EmitterState()
   &SpawnRate(SpawnRate=50, Use Spawn Probability=false) {module=SpawnRate001}
   &SpawnRate(
    Spawn Probability=&MakeCustomFloatFromBool(Bool=&SetBoolByFloatComparison(A=$Emitter.LoopedAge, B=899), True Float=0, False Float=1)
    SpawnRate=&Waveform(
     "Waveform [1]"=::NewEnumerator0
     "[1] Frequency"=1
     "[1] Compound Frequency"=1
     "[1] Amplitude Min/Max"=&Lerp_Vector2(A=<2 4>, B=<6 10>, Alpha=$Emitter.NormalizedLoopAge)
    )
    Use Spawn Probability=true
   ) {enabled=false}
   &SpawnBurst_Instantaneous(Spawn Group=1, Spawn Time=899.75, Use Spawn Probability=false, UseLoopCountLimit=false) {enabled=false}
   &DoOnce(
    Reset:NiagaraBool=&SetBoolByFloatComparison(A=$Emitter.NormalizedLoopAge, B=0.0001, Comparison Type=::NewEnumerator7)
    Trigger Condition:NiagaraBool=&SetBoolByFloatComparison(A=$Emitter.NormalizedLoopAge, B=0.999, Comparison Type=::NewEnumerator2)
   )
   &SpawnPerFrame(Spawn:NiagaraBool=$Emitter.DoOnce.Execute, Spawn Group=1, Use Spawn Probability=false)
  ]
  particleSpawnScript [
   &InitializeParticle(
    Color=&MaskLinearColorBySpawnGroup(
     Masked Linear Color=<rgba 10 9.3074 4.125 1>
     Passthrough Linear Color=&LerpLinearColors(
      StartColor=<rgba 0.5 0.9337 1 1>
      EndColor=<rgba 8 7.4459 3.3 1>
      LerpFactor=&FloatFromCurve(
       FloatCurve=DataInterface() {propertyValues {Curve {keys [
           {arriveTangent=0.5002, interpMode=RCIM_Cubic, leaveTangent=0.5002, tangentMode=RCTM_User, time=0, value=0}
           {
            arriveTangent=0.0672
            arriveTangentWeight=0.0749
            interpMode=RCIM_Cubic
            leaveTangent=0.4983
            leaveTangentWeight=0.4892
            tangentMode=RCTM_Break
            tangentWeightMode=RCTWM_WeightedBoth
            time=0.1
            value=0.02
           }
           {
            arriveTangent=2.9395
            arriveTangentWeight=0.8473
            interpMode=RCIM_Cubic
            leaveTangent=3.3978
            leaveTangentWeight=0.9847
            tangentMode=RCTM_Break
            tangentWeightMode=RCTWM_WeightedBoth
            time=1
            value=0.98
           }
          ]}}}
       CurveIndex=$Emitter.NormalizedLoopAge
      )
     )
     Spawn Group Mask=1
    )
    Lifetime=&MaskFloatBySpawnGroup(Masked Float=1, Passthrough Float=&RandomRangeFloat(Minimum=0.5, Maximum=0.35), Spawn Group Mask=1)
    Position Offset=<0 70 210>
    Sprite Rotation Angle=&RandomRangeFloat(Minimum=5, Maximum=-5)
    Sprite Rotation Mode=::NewEnumerator0
    Sprite Size Mode=::NewEnumerator3
    Uniform Sprite Size=&MaskFloatBySpawnGroup(Masked Float=5, Passthrough Float=&RandomRangeFloat(Minimum=10, Maximum=10), Spawn Group Mask=1)
    UsePositionOffset=false
    Write Material Random=true
   )
   &AddVelocityInCone(
    Cone Angle=0
    Cone Axis=<0 0 1>
    Cone Axis Coordinate Space=::World
    Module.Use Velocity Falloff On Cone Axis=false
    Velocity Distribution Along Cone Axis=0.75
    Velocity Strength=&MaskFloatBySpawnGroup(Masked Float=250, Passthrough Float=&RandomRangeFloat(Minimum=220, Maximum=180), Spawn Group Mask=1)
   ) {enabled=false}
   /SetVariables(Particles.RollingDigits:NiagaraInt32=&MaskIntBySpawnGroup(
     Masked Int=$System.MaxNumRange
     Passthrough Int=&MakeIntFromFloat(Float=&FloatFromCurve(
       FloatCurve=DataInterface() {propertyValues {Curve {keys [
           {
            arriveTangentWeight=1.0752
            interpMode=RCIM_Cubic
            leaveTangentWeight=1.0752
            tangentMode=RCTM_User
            tangentWeightMode=RCTWM_WeightedBoth
            time=0
            value=0
           }
           {
            arriveTangent=7.3251
            arriveTangentWeight=0.879
            interpMode=RCIM_Cubic
            leaveTangent=7.3251
            leaveTangentWeight=0.879
            tangentMode=RCTM_User
            tangentWeightMode=RCTWM_WeightedBoth
            time=1
            value=1
           }
          ]}}}
       CurveIndex=$Emitter.NormalizedLoopAge
       Scale Curve=&MakeFloatFromInt(Int=$System.MaxNumRange)
      ))
     Spawn Group Mask=1
    )) {isSetParameters=true}
   /Niagara/Modules/Audio/PlayPersistentAudio(Audio Player:NiagaraDataInterfaceAudioPlayer=$User.AudioPlayer, Play Audio=true) {enabled=false}
   &ShapeLocation(MaskSpawnGroup=false, Override Seed=false, Sphere Radius=1, Sphere Surface Distribution=1, UseOwnerScale=true)
  ]
  particleUpdateScript [
   &ParticleState()
   &ScaleSpriteSize(
    Scale Sprite Size Mode=::NewEnumerator0
    Uniform Scale Factor=&MaskFloatBySpawnGroup(
     Masked Float=&FloatFromCurve(FloatCurve=DataInterface() {propertyValues {Curve {keys [
          {interpMode=RCIM_Cubic, leaveTangentWeight=0.0698, tangentMode=RCTM_User, time=0, value=0}
          {arriveTangentWeight=0.1582, interpMode=RCIM_Cubic, leaveTangentWeight=0.1204, tangentMode=RCTM_User, time=0.2, value=1.5}
          {arriveTangentWeight=0.0695, interpMode=RCIM_Cubic, leaveTangentWeight=0.1889, tangentMode=RCTM_User, time=0.35, value=1}
          {arriveTangentWeight=0.1727, interpMode=RCIM_Cubic, leaveTangentWeight=0.0399, tangentMode=RCTM_User, time=0.8015, value=1}
          {arriveTangentWeight=0.1303, interpMode=RCIM_Cubic, leaveTangentWeight=0.1283, tangentMode=RCTM_User, time=0.9015, value=1.35}
          {arriveTangentWeight=0.0328, interpMode=RCIM_Cubic, tangentMode=RCTM_User, time=1, value=0}
         ]}}})
     Passthrough Float=&FloatFromCurve(FloatCurve=DataInterface() {propertyValues {Curve {keys [
          {arriveTangent=7.2698, interpMode=RCIM_Cubic, leaveTangent=7.2698, tangentMode=RCTM_User, time=0, value=0.25}
          {arriveTangentWeight=0.1895, interpMode=RCIM_Cubic, leaveTangentWeight=0.1441, tangentMode=RCTM_User, time=0.2093, value=1.35}
          {arriveTangentWeight=0.0617, interpMode=RCIM_Cubic, leaveTangentWeight=0.1676, tangentMode=RCTM_User, time=0.3685, value=1}
          {arriveTangent=0.4691, interpMode=RCIM_Cubic, leaveTangent=0.4691, time=0.8015, value=1}
          {arriveTangentWeight=0.1303, interpMode=RCIM_Cubic, leaveTangentWeight=0.1283, tangentMode=RCTM_User, time=0.9015, value=1.25}
          {arriveTangent=-7.3489, interpMode=RCIM_Cubic, leaveTangent=-7.3489, tangentMode=RCTM_User, time=1, value=0}
         ]}}})
     Spawn Group Mask=1
    )
   ) {enabled=false}
   &self:NewCalcDigits(MaximumNumber:NiagaraInt32=$Particles.RollingDigits) {module=NewCalcDigits}
   /SetVariables(
    Particles.PivotOffset 1:Vector2f=&MakeVector2D(
     X=&Add_Float(
      A=&FloatFromCurve(FloatCurve=$Emitter.Overall Offset, CurveIndex=$Particles.Testdigits)
      B=&Lerp_Float(A=0, B=-4.5, Alpha=&FloatFromCurve(FloatCurve=$Emitter.Space Curve, CurveIndex=$Particles.Testdigits))
     )
     Y=0.5
    )
    Particles.PivotOffset 10:Vector2f=&MakeVector2D(
     X=&Add_Float(
      A=&FloatFromCurve(FloatCurve=$Emitter.Overall Offset, CurveIndex=$Particles.Testdigits)
      B=&Lerp_Float(A=0, B=4.5, Alpha=&FloatFromCurve(FloatCurve=$Emitter.Space Curve, CurveIndex=$Particles.Testdigits))
     )
     Y=0.5
    )
    Particles.PivotOffset 2:Vector2f=&MakeVector2D(
     X=&Add_Float(
      A=&FloatFromCurve(FloatCurve=$Emitter.Overall Offset, CurveIndex=$Particles.Testdigits)
      B=&Lerp_Float(A=0, B=-3.5, Alpha=&FloatFromCurve(FloatCurve=$Emitter.Space Curve, CurveIndex=$Particles.Testdigits))
     )
     Y=0.5
    )
    Particles.PivotOffset 3:Vector2f=&MakeVector2D(
     X=&Add_Float(
      A=&FloatFromCurve(FloatCurve=$Emitter.Overall Offset, CurveIndex=$Particles.Testdigits)
      B=&Lerp_Float(A=0, B=-2.5, Alpha=&FloatFromCurve(FloatCurve=$Emitter.Space Curve, CurveIndex=$Particles.Testdigits))
     )
     Y=0.5
    )
    Particles.PivotOffset 4:Vector2f=&MakeVector2D(
     X=&Add_Float(
      A=&FloatFromCurve(FloatCurve=$Emitter.Overall Offset, CurveIndex=$Particles.Testdigits)
      B=&Lerp_Float(A=0, B=-1.5, Alpha=&FloatFromCurve(FloatCurve=$Emitter.Space Curve, CurveIndex=$Particles.Testdigits))
     )
     Y=0.5
    )
    Particles.PivotOffset 5:Vector2f=&MakeVector2D(
     X=&Add_Float(
      A=&FloatFromCurve(FloatCurve=$Emitter.Overall Offset, CurveIndex=$Particles.Testdigits)
      B=&Lerp_Float(A=0, B=-0.5, Alpha=&FloatFromCurve(FloatCurve=$Emitter.Space Curve, CurveIndex=$Particles.Testdigits))
     )
     Y=0.5
    )
    Particles.PivotOffset 6:Vector2f=&MakeVector2D(
     X=&Add_Float(
      A=&FloatFromCurve(FloatCurve=$Emitter.Overall Offset, CurveIndex=$Particles.Testdigits)
      B=&Lerp_Float(A=0, B=0.5, Alpha=&FloatFromCurve(FloatCurve=$Emitter.Space Curve, CurveIndex=$Particles.Testdigits))
     )
     Y=0.5
    )
    Particles.PivotOffset 7:Vector2f=&MakeVector2D(
     X=&Add_Float(
      A=&FloatFromCurve(FloatCurve=$Emitter.Overall Offset, CurveIndex=$Particles.Testdigits)
      B=&Lerp_Float(A=0, B=1.5, Alpha=&FloatFromCurve(FloatCurve=$Emitter.Space Curve, CurveIndex=$Particles.Testdigits))
     )
     Y=0.5
    )
    Particles.PivotOffset 8:Vector2f=&MakeVector2D(
     X=&Add_Float(
      A=&FloatFromCurve(FloatCurve=$Emitter.Overall Offset, CurveIndex=$Particles.Testdigits)
      B=&Lerp_Float(A=0, B=2.5, Alpha=&FloatFromCurve(FloatCurve=$Emitter.Space Curve, CurveIndex=$Particles.Testdigits))
     )
     Y=0.5
    )
    Particles.PivotOffset 9:Vector2f=&MakeVector2D(
     X=&Add_Float(
      A=&FloatFromCurve(FloatCurve=$Emitter.Overall Offset, CurveIndex=$Particles.Testdigits)
      B=&Lerp_Float(A=0, B=3.5, Alpha=&FloatFromCurve(FloatCurve=$Emitter.Space Curve, CurveIndex=$Particles.Testdigits))
     )
     Y=0.5
    )
   ) {enabled=false, isSetParameters=true}
   &Drag(Drag=5, Use Linear Drag=true, Use Rotational Drag=false) {enabled=false}
   /SetVariables(
    Particles.Heart Color:LinearColor=&LerpLinearColors(
     StartColor=&HueShiftLinearColor(Hue=&RandomRangeFloat(Minimum=-0.01, Maximum=0.01), Saturation=1, Value=1, Linear Color=<rgba 1 0 0 1>)
     EndColor=&HueShiftLinearColor(Hue=&RandomRangeFloat(Minimum=-0.01, Maximum=0.01), Saturation=1, Value=1, Linear Color=<rgba 1 0 0 1>)
     LerpFactor=&FloatFromCurve(
      FloatCurve=DataInterface() {propertyValues {Curve {keys [
          {arriveTangent=0.5002, interpMode=RCIM_Cubic, leaveTangent=0.5002, tangentMode=RCTM_User, time=0, value=0}
          {
           arriveTangent=0.0672
           arriveTangentWeight=0.0749
           interpMode=RCIM_Cubic
           leaveTangent=0.1039
           leaveTangentWeight=0.2528
           tangentMode=RCTM_Break
           tangentWeightMode=RCTWM_WeightedBoth
           time=0.1
           value=0.02
          }
          {
           arriveTangent=23.2766
           arriveTangentWeight=1.0032
           interpMode=RCIM_Cubic
           leaveTangent=3.3978
           leaveTangentWeight=0.9847
           tangentMode=RCTM_Break
           tangentWeightMode=RCTWM_WeightedBoth
           time=1
           value=1
          }
         ]}}}
      CurveIndex=$Emitter.NormalizedLoopAge
     )
    )
    Particles.Heart Pivot:Vector2f=&MakeVector2D(
     X=&FloatFromCurve(
      FloatCurve=DataInterface() {propertyValues {Curve {keys [<at 1 1.2906>,<at 2 1.3429>,<at 3 1.6446>,<at 4 2.0196>,<at 5 2.2626>,<at 6 2.44>,<at 7 2.7133>,<at 8 2.875>,<at 9 3.125>,<at 10.0059 3.2451>]}}}
      CurveIndex=$Particles.Testdigits
     )
     Y=0.5
    )
    Particles.Heart Size:Vector2f=&Multiply_Vector2D_ByFloat(Float=2, Vector2D=$Particles.SpriteSize)
   ) {enabled=false, isSetParameters=true}
   &InheritSourceMovement(ApplyForce=false, ApplyPosition=true, ApplyPositionFalloff=false) {enabled=false}
   &SolveForcesAndVelocity(Clamp Velocity=false, Limit Acceleration=false) {enabled=false}
   /Niagara/Modules/Audio/UpdatePersistentAudio(
    Audio Player:NiagaraDataInterfaceAudioPlayer=$User.AudioPlayer
    Bool Value:NiagaraBool=&SetBoolByIntComparison(A=&SpawnGroup(), B=1, Comparison Type=::NewEnumerator4)
    Update Type:ENiagara_AudioParamType=::NewEnumerator0
   ) {enabled=false}
   &self:GimbalLock(Gimbal Lock Direction:NiagaraPosition=&ReturnCameraVector(Camera Vectors=::NewEnumerator4, Coordinate Space=::Local, Origin Position=$Engine.Emitter.SimulationPosition)) {module=GimbalLock}
  ]
  rendererClasses [&NiagaraSpriteRendererProperties]
  renderers [
   {rendererClass=&NiagaraSpriteRendererProperties}
   {rendererClass=&NiagaraSpriteRendererProperties, rendererIndex=1}
   {rendererClass=&NiagaraSpriteRendererProperties, rendererIndex=2}
   {rendererClass=&NiagaraSpriteRendererProperties, rendererIndex=3}
   {rendererClass=&NiagaraSpriteRendererProperties, rendererIndex=4}
   {rendererClass=&NiagaraSpriteRendererProperties, rendererIndex=5}
   {rendererClass=&NiagaraSpriteRendererProperties, rendererIndex=6}
   {rendererClass=&NiagaraSpriteRendererProperties, rendererIndex=7}
   {rendererClass=&NiagaraSpriteRendererProperties, rendererIndex=8}
   {rendererClass=&NiagaraSpriteRendererProperties, rendererIndex=9}
   {rendererClass=&NiagaraSpriteRendererProperties, rendererIndex=10}
  ]
 }}
summary {
 systemName=NS_Emote_QuarterClaspJoin1
 emitters [{emitterName=Counter_nosim_infDigits001, rendererClasses [&NiagaraSpriteRendererProperties]}]
 userVariables [
  User.AudioPlayer {
   type=NiagaraDataInterfaceAudioPlayer
   description=""
   defaultValue {
    struct=NiagaraExt_VariableValue_DataInterface
    value {
     dataInterface=&self:NiagaraDataInterfaceAudioPlayer_0
     dataInterfaceClass=&NiagaraDataInterfaceAudioPlayer
    }
   }
  }
  User.MaximumNumber {type=NiagaraInt32, defaultValue=9999999}
 ]
}
